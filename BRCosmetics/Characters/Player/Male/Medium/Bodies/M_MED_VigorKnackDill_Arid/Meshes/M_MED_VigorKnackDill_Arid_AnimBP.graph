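# Animation Blueprint for the M_MED_VigorKnackDill_Arid player body cosmetic, following the attached parent mesh's pose and adding secondary zipper motion.
# The copied pose gets an additive scale on corr_chest, then dyn_zipper_constr is constrained to weighted spine_02 through spine_05 bones.
# The body's Control Rig runs next, feeding an Anim Dynamics node that swings dyn_zipper with gravity, an angular spring and angular limits.
# The event graph holds no logic.
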AnimBlueprint /BRCosmetics/Characters/Player/Male/Medium/Bodies/M_MED_VigorKnackDill_Arid/Meshes/M_MED_VigorKnackDill_Arid_AnimBP
target_skeleton=&/M_MED_VigorKnackDill_Arid_Skeleton
graphs [
 AnimGraph {nodes [
   n1 = Copy Pose From Mesh() {bUseAttachedParent=true}
   n2 = "Constraint - dyn_zipper_constr (PPPP)"!(
    ComponentPose="Transform (Modify) Bone - Bone: corr_chest"(ComponentPose=Local To Component(LocalPose=@n1), Alpha=1, Translation="0,0,0", Rotation="0,0,0", Scale="0.1,0.1,0.1") {scaleMode=BMM_Additive, boneToModify {boneName=corr_chest}, scale=<0.1 0.1 0.1>}
    Alpha=1
    ConstraintWeights_0=0.0754
    ConstraintWeights_1:float=0.1893
    ConstraintWeights_2:float=0.4925
    ConstraintWeights_3:float=0.2428
   ) -> (Pose) {
    boneToModify {boneName=dyn_zipper_constr}
    constraintSetup [{targetBone {boneName=spine_02}}, {targetBone {boneName=spine_03}}, {targetBone {boneName=spine_04}}, {targetBone {boneName=spine_05}}]
    constraintWeights [1,1,1,1]
   }
   n3 = Control Rig(Source=Component To Local(ComponentPose=@n2)) {controlRigAssetReference {blueprintRigClass=&/M_MED_VigorKnackDill_Arid_CtrlRig.M_MED_VigorKnackDill_Arid_CtrlRig_C}}
   n4 = Anim Dynamics - Bone: dyn_zipper!(ComponentPose=Local To Component(LocalPose=@n3), Alpha=1) -> (Pose) {
    angularDampingOverride=0.999
    angularSpringConstant=500
    bAngularSpring=true
    bOverrideAngularDamping=true
    bUseGravityOverride=true
    bUsePlanarLimit=false
    simulationSpace=RootRelative
    boundBone {boneName=dyn_zipper}
    componentAppliedLinearAccClamp=<10000 10000 10000>
    componentLinearAccScale=<1 1 1>
    componentLinearVelScale=<1 1 1>
    gravityOverride {z=-980}
    physicsBodyDefinitions [{
      boundBone {boneName=dyn_zipper}
      boxExtents=<10 10 10>
      constraintSetup {angularLimitsMax=<0 0 5>, angularLimitsMin=<0 -45 -5>, angularTarget=<1 0 0>}
      localJointOffset=<0.2 0 0>
     }]
    simSpaceSettings {maxAngularAcceleration=35, maxAngularVelocity=7, simSpaceAngularAlpha=1}
   }
   Output Pose(Result=Component To Local(ComponentPose=@n4)) {layerGroup=None}
  ]}
 EventGraph {nodes [Event BlueprintUpdateAnimation()]}
]
variables []
