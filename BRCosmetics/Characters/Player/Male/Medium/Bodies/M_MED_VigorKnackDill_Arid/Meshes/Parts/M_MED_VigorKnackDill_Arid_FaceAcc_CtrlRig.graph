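# Control Rig for the VigorKnackDill Arid face accessory, correcting hood and body deform bones.
# Construction imports the skeleton and adds neck/head hood helper bones.
# Forward solve runs shared DeformRig, applies rotation-driven clavicle correctives, then feeds smoothed half head twist into both neck bones, restoring head from its saved copy.
# Hood corrective bones blend between head and neck via parent constraints, with extra offsets driven by head Z-axis rotation against neck_02.
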ControlRigBlueprint /BRCosmetics/Characters/Player/Male/Medium/Bodies/M_MED_VigorKnackDill_Arid/Meshes/Parts/M_MED_VigorKnackDill_Arid_FaceAcc_CtrlRig
graphs [
 Add_1_ContainedGraph {nodes [
   n1 = RigVMFunctionEntryNode()
   Add = &RigVMFunction_MathDoubleAdd(A=@n1.A, B=@n1.B)
   RigVMFunctionReturnNode(Result=&RigVMFunction_MathDoubleAdd(A=@Add.Result, B=@n1.C).Result)
  ]}
 New Function_ContainedGraph {nodes [
   n1 = RigVMFunctionEntryNode()
   RigVMFunction_MathBoolToInteger = &RigVMFunction_MathBoolToInteger(Value=@n1.twistOnly)
   RigVMFunction_MathBoolOr = &RigVMFunction_MathBoolOr(B=@n1.Parent Initial, A=@n1.fromSelf)
   RigVMFunction_MathQuaternionSwingTwist = &RigVMFunction_MathQuaternionSwingTwist(
    Input=&RigUnit_GetRelativeTransformForItem(
     bChildInitial=False
     Child=@n1.Node
     Parent=RigVMDispatchNode(True=@n1.Node, False=@n1.altParent, Condition=@n1.fromSelf) {template="DISPATCH_RigVMDispatch_If(in Condition,in True,in False,out Result)"}.Result
     bParentInitial=@RigVMFunction_MathBoolOr.Result
    ).RelativeTransform.Rotation
    TwistAxis=@n1.twistAxis
   )
   RigVMFunction_MathQuaternionToEuler = &RigVMFunction_MathQuaternionToEuler(
    Value=RigVMDispatchNode(
     Values=((),())
     Index=@RigVMFunction_MathBoolToInteger.Result
     Values.0=@RigVMFunction_MathQuaternionSwingTwist.Swing
     Values.1=@RigVMFunction_MathQuaternionSwingTwist.Twist
    ) {template="DISPATCH_RigVMDispatch_SelectInt32(in Index,in Values,out Result)"}.Result
    RotationOrder=@n1.RotOrder
   )
   > DISPATCH_RigVMDispatch_Print = RigVMDispatchNode(
    Prefix="Relative rotation "
    ScreenColor=(R=1,G=1,B=1,A=1)
    ScreenDuration=0.01
    Value=@RigVMFunction_MathQuaternionToEuler.Result
    Enabled=@n1.debug
   ) {template="DISPATCH_RigVMDispatch_Print(in Prefix,in Value,in Enabled,in ScreenDuration,in ScreenColor)"}
   RigVMFunction_MathVectorLengthSquared = /Script/RigVM.RigVMFunction_MathVectorLengthSquared(Value=@n1.watchAxis) {is=struct RigVMUnitNode, template="LengthSquared::Execute(in Value,out Result)"}
   RigVMFunction_MathVectorFromFloat = /Script/RigVM.RigVMFunction_MathVectorFromFloat(Value=@RigVMFunction_MathVectorLengthSquared.Result) {is=struct RigVMUnitNode, template="Cast::Execute(in Value,out Result)"}
   Divide = /Script/RigVM.RigVMFunction_MathVectorDiv(A=@RigVMFunction_MathQuaternionToEuler.Result, B=@RigVMFunction_MathVectorFromFloat.Result) {is=struct RigVMUnitNode, template="Divide::Execute(in A,in B,out Result)"}
   n2 = /Script/RigVM.RigVMFunction_MathVectorMul(A=@Divide.Result, B=@n1.watchAxis) {is=struct RigVMUnitNode, template="Multiply::Execute(in A,in B,out Result)"}
   Add = RigVMAggregateNode(A=@n2.Result.X, B=@n2.Result.Y, C=@n2.Result.Z)
   > RigVMDispatchNode(Prefix="delta on watchAxis ", ScreenColor=(R=1,G=1,B=1,A=1), ScreenDuration=0.01, Value=@Add.Result, Enabled=@n1.debug) {template="DISPATCH_RigVMDispatch_Print(in Prefix,in Value,in Enabled,in ScreenDuration,in ScreenColor)"}
   Remap = &RigVMFunction_MathDoubleRemap(TargetMaximum=1, TargetMinimum=0, Value=@Add.Result, SourceMaximum=@n1.max, SourceMinimum=@n1.min)
   Multiply = &RigVMFunction_MathDoubleMul(A=@Remap.Result, B=@n1.Alpha Premult)
   > RigVMFunctionReturnNode(Delta on watchAxis=@Add.Result, outAlpha=@Multiply.Result)
  ]}
 RigVMFunctionLibrary {nodes [DeltaRef_rotation = RigVMCollapseNode(
    Alpha Premult=1
    ExecuteContext=()
    Node=(Type=None,Name="None")
    Parent Initial=true
    RotOrder=ZYX
    altParent=(Type=Bone,Name="None")
    debug=True
    fromSelf=False
    max=1
    min=0
    twistAxis=(X=1,Y=0,Z=0)
    twistOnly=False
    watchAxis=(X=0,Y=0,Z=0)
   )]}
 RigVMFunction_Sequence_2_ContainedGraph {nodes [
   n1 = RigVMFunctionEntryNode()
   n2 = &RigVMFunction_Sequence(ExecuteContext=@n1.ExecuteContext)
   n3 = &RigVMFunction_Sequence(ExecuteContext=@n2.B)
   RigVMFunctionReturnNode(A=@n2.A, B=@n3.A, C=@n3.B)
  ]}
 RigVMFunction_Sequence_4_ContainedGraph {nodes [
   n1 = RigVMFunctionEntryNode()
   n2 = &RigVMFunction_Sequence(ExecuteContext=@n1.ExecuteContext)
   n3 = &RigVMFunction_Sequence(ExecuteContext=@n2.B)
   n4 = &RigVMFunction_Sequence(ExecuteContext=@n3.B)
   RigVMFunctionReturnNode(B=@n3.A, A=@n2.A, C=@n4.A, D=@n4.B)
  ]}
 RigVMFunction_Sequence_5_ContainedGraph {nodes [
   n1 = RigVMFunctionEntryNode()
   RigVMFunction_Sequence = &RigVMFunction_Sequence(ExecuteContext=@n1.ExecuteContext)
   n2 = &RigVMFunction_Sequence(ExecuteContext=@RigVMFunction_Sequence.B)
   RigVMFunctionReturnNode(A=@RigVMFunction_Sequence.A, B=@n2.A, C=@n2.B)
  ]}
 RigVMModel {nodes [
   n1 = &RigUnit_PrepareForExecution()
   &RigUnit_HierarchyImportFromSkeleton(NameSpace=None, bIncludeCurves=true, bIncludeMeshSockets=False, bIncludeVirtualBones=True, ExecutePin=@n1.ExecutePin)
   RigUnit_GetTransform = &RigUnit_GetTransform(Item=(Type=Bone,Name="neck_01"), bInitial=true)
   > n2 = /Script/ControlRig.RigUnit_HierarchyAddBone(Name=neck_01_hood, Parent=(Type=Bone,Name="spine_05"), Space=GlobalSpace, Transform=@RigUnit_GetTransform.Transform) {is=struct RigVMUnitNode}
   > &RigUnit_HierarchyAddBone(
    Name=neck_02_hood
    Space=GlobalSpace
    Parent=@n2.Item
    Transform=&RigUnit_GetTransform(Item=(Type=Bone,Name="neck_02"), bInitial=true).Transform
   )
   > &RigUnit_HierarchyAddBone(
    Name=head_hood
    Parent=(Type=Bone,Name="spine_05")
    Space=GlobalSpace
    Transform=&RigUnit_GetTransform(Item=(Type=Bone,Name="head"), bInitial=true).Transform
   )
   RigUnit_BeginExecution = &RigUnit_BeginExecution()
   FN_DeformRig_FNC = RigVMFunctionReferenceNode(
    DeformRig_ Curve="(deformRig_l_upperarm_up_alpha_72_#1=False,deformRig_r_upperarm_up_alpha_73_#2=False,deformRig_l_upperarm_fr_alpha_74_#3=False,deformRig_r_upperarm_fr_alpha_75_CB51B8 ...(truncated)"
    DeformRig_Bool="(upperarm_1_#4=True,elbow_3_#5=True,wrist_7_#6=True,groin_8_#7=True,glute_17_#8=True,knee_9_40C408324CA261 ...(truncated)"
    DeformRig_Debug=(upperarm_debug_18_#4=False,wrist_debug_20_#6=False,lat_debug_26_#9=False,groinGlute_debug_27_#7=False)
    DeformRig_Value="(upperarm_up_18_#10=(Rotation=(X=0,Y=0,Z=0,W=1),Translation=(X=0,Y=0,Z=8),Scale3D=(X=0.7,Y=1,Z=1)),upperarm_fwd_fr_19_#11=(Rotation=(X=0,Y=0,Z=0,W=1),Translation=(X=0,Y=10,Z=0),Scale3D=(X=0.5,Y=1,Z= ...(truncated)"
    ExecuteContext=@RigUnit_BeginExecution.ExecutePin
   )
   > RigVMFunction_Sequence = RigVMAggregateNode()
   DeltaRef_rotation = RigVMFunctionReferenceNode(
    Alpha Premult=1
    Node=(Type=Bone,Name="clavicle_l")
    Parent Initial=true
    RotOrder=ZYX
    altParent=(Type=Bone,Name="None")
    debug=false
    fromSelf=true
    max=-27
    min=0
    twistAxis=(X=1,Y=0,Z=0)
    twistOnly=False
    watchAxis=(X=0,Y=1,Z=0)
    ExecuteContext=@RigVMFunction_Sequence.A
   )
   > &RigUnit_ModifyTransforms(
    ItemToModify=((Item=(Type=Bone,Name="def_clavicle_l"),Transform=(Rotation=(X=0,Y=0.2804,Z=0,W=0.9599),Translation=(X=6.5121,Y=0,Z=0.2302),Scale3D=(X=1,Y=1,Z=1))))
    WeightMinimum=0
    Weight=@DeltaRef_rotation.outAlpha
   )
   DeltaRef_rotation_1 = RigVMFunctionReferenceNode(
    Alpha Premult=1
    Node=(Type=Bone,Name="clavicle_r")
    Parent Initial=true
    RotOrder=ZYX
    altParent=(Type=Bone,Name="None")
    debug=false
    fromSelf=true
    max=-27
    min=0
    twistAxis=(X=1,Y=0,Z=0)
    twistOnly=False
    watchAxis=(X=0,Y=1,Z=0)
    ExecuteContext=@RigVMFunction_Sequence.B
   )
   > &RigUnit_ModifyTransforms(
    ItemToModify=((Item=(Type=Bone,Name="def_clavicle_r"),Transform=(Rotation=(X=0,Y=0.2804,Z=0,W=0.9599),Translation=(X=6.5121,Y=0,Z=0.2302),Scale3D=(X=1,Y=1,Z=1))))
    WeightMinimum=0
    Weight=@DeltaRef_rotation_1.outAlpha
   )
   RigVMFunction_Sequence_2 = RigVMAggregateNode(ExecuteContext=@RigVMFunction_Sequence.C)
   &RigUnit_ParentConstraint(
    AdvancedSettings=(InterpolationType=Average,RotationOrderForFilter=XZY)
    Child=(Type=Bone,Name="neck_01_hood")
    Filter=(TranslationFilter=(bX=True,bY=True,bZ=True),RotationFilter=(bX=false,bY=True,bZ=True),ScaleFilter=(bX=True,bY=True,bZ=True))
    Parents=((Item=(Type=Bone,Name="neck_01"),Weight=1))
    Weight=1
    ExecutePin=@RigVMFunction_Sequence_2.A
   )
   > &RigUnit_ParentConstraint(
    AdvancedSettings=(InterpolationType=Average,RotationOrderForFilter=XZY)
    Child=(Type=Bone,Name="neck_02_hood")
    Filter=(TranslationFilter=(bX=True,bY=True,bZ=True),RotationFilter=(bX=false,bY=True,bZ=True),ScaleFilter=(bX=True,bY=True,bZ=True))
    Parents=((Item=(Type=Bone,Name="neck_02"),Weight=1))
    Weight=1
   )
   > RigUnit_SetTransform = &RigUnit_SetTransform(
    Item=(Type=Bone,Name="head_hood")
    Weight=1
    bInitial=False
    Value=&RigUnit_GetTransform(Item=(Type=Bone,Name="head"), bInitial=False).Transform
   )
   DeltaRef_rotation_5 = RigVMFunctionReferenceNode(
    Alpha Premult=1
    Node=(Type=Bone,Name="head")
    Parent Initial=false
    RotOrder=ZYX
    altParent=(Type=Bone,Name="spine_05")
    debug=false
    fromSelf=false
    max=1
    min=0
    twistAxis=(X=1,Y=0,Z=0)
    twistOnly=true
    watchAxis=(X=1,Y=0,Z=0)
    ExecuteContext=@RigVMFunction_Sequence_2.B
   )
   Divide = &RigVMFunction_MathDoubleDiv(B=2, A=@DeltaRef_rotation_5.Delta on watchAxis)
   AlphaInterp = &RigVMFunction_AlphaInterp(InterpSpeedDecreasing=15, InterpSpeedIncreasing=15, Scale=1, bClampResult=False, Value=@Divide.Result)
   RigVMFunction_MathQuaternionFromEuler = &RigVMFunction_MathQuaternionFromEuler(Euler=(X=0,Y=0,Z=0), Euler.X=@AlphaInterp.Result)
   > &RigUnit_ModifyTransforms(
    ItemToModify=((Item=(Type=Bone,Name="neck_01"),Transform=(Rotation=(X=0,Y=0,Z=0,W=1),Translation=(X=0,Y=0,Z=0),Scale3D=(X=1,Y=1,Z=1))),(Item=(Type=Bone,Name="neck_02"),Transform=(Rotation=(X=0,Y=0,Z=0,W=1),Translation=(X=0,Y=0,Z=0),Scale3D=(X=1,Y=1,Z=1))))
    Weight=1
    WeightMinimum=0
    ItemToModify.0.Transform.Rotation=@RigVMFunction_MathQuaternionFromEuler.Result
    ItemToModify.1.Transform.Rotation=@RigVMFunction_MathQuaternionFromEuler.Result
   )
   > &RigUnit_ParentConstraint(
    AdvancedSettings=(InterpolationType=Average,RotationOrderForFilter=XZY)
    Child=(Type=Bone,Name="head")
    Parents=((Item=(Type=Bone,Name="head_hood"),Weight=1))
    Weight=1
   )
   RigVMFunction_Sequence_6 = RigVMAggregateNode(ExecuteContext=@RigVMFunction_Sequence_2.C)
   n3 = RigVMVariableNode() {variable=hoodConstr_val_01}
   n4 = RigVMVariableNode() {variable=hoodConstr_val_02}
   &RigUnit_ParentConstraint(
    AdvancedSettings=(InterpolationType=Average,RotationOrderForFilter=XZY)
    Child=(Type=Bone,Name="dyn_hood_bk_up_r_corr")
    Parents=((Item=(Type=Bone,Name="head"),Weight=0.1),(Item=(Type=Bone,Name="neck_02"),Weight=0.9))
    Weight=1
    ExecutePin=@RigVMFunction_Sequence_6.A
    Parents.0.Weight=@n3.Value
    Parents.1.Weight=@n4.Value
   )
   > &RigUnit_ParentConstraint(
    AdvancedSettings=(InterpolationType=Average,RotationOrderForFilter=XZY)
    Child=(Type=Bone,Name="dyn_hood_bk_up_corr")
    Parents=((Item=(Type=Bone,Name="head"),Weight=0.1),(Item=(Type=Bone,Name="neck_02"),Weight=0.9))
    Weight=1
    Parents.0.Weight=@n3.Value
    Parents.1.Weight=@n4.Value
   )
   > &RigUnit_ParentConstraint(
    AdvancedSettings=(InterpolationType=Average,RotationOrderForFilter=XZY)
    Child=(Type=Bone,Name="dyn_hood_bk_up_l_corr")
    Parents=((Item=(Type=Bone,Name="head"),Weight=0.1),(Item=(Type=Bone,Name="neck_02"),Weight=0.9))
    Weight=1
    Parents.0.Weight=@n3.Value
    Parents.1.Weight=@n4.Value
   )
   > &RigUnit_ParentConstraint(
    AdvancedSettings=(InterpolationType=Average,RotationOrderForFilter=XZY)
    Child=(Type=Bone,Name="dyn_hood_fr_up_r_corr")
    Parents=((Item=(Type=Bone,Name="head"),Weight=0.7),(Item=(Type=Bone,Name="neck_02"),Weight=0.3))
    Weight=1
   )
   > &RigUnit_ParentConstraint(
    AdvancedSettings=(InterpolationType=Average,RotationOrderForFilter=XZY)
    Child=(Type=Bone,Name="dyn_hood_fr_up_l_corr")
    Parents=((Item=(Type=Bone,Name="head"),Weight=0.7),(Item=(Type=Bone,Name="neck_02"),Weight=0.3))
    Weight=1
   )
   &RigUnit_ParentConstraint(
    AdvancedSettings=(InterpolationType=Average,RotationOrderForFilter=XZY)
    Child=(Type=Bone,Name="dyn_hood_bk_r_corr")
    Parents=((Item=(Type=Bone,Name="neck_02"),Weight=0.1),(Item=(Type=Bone,Name="neck_01"),Weight=0.9))
    Weight=1
    ExecutePin=@RigVMFunction_Sequence_6.B
    Parents.0.Weight=@n3.Value
    Parents.1.Weight=@n4.Value
   )
   > &RigUnit_ParentConstraint(
    AdvancedSettings=(InterpolationType=Average,RotationOrderForFilter=XZY)
    Child=(Type=Bone,Name="dyn_hood_bk_corr")
    Parents=((Item=(Type=Bone,Name="neck_02"),Weight=0.1),(Item=(Type=Bone,Name="neck_01"),Weight=0.9))
    Weight=1
    Parents.0.Weight=@n3.Value
    Parents.1.Weight=@n4.Value
   )
   > &RigUnit_ParentConstraint(
    AdvancedSettings=(InterpolationType=Average,RotationOrderForFilter=XZY)
    Child=(Type=Bone,Name="dyn_hood_bk_l_corr")
    Parents=((Item=(Type=Bone,Name="neck_02"),Weight=0.1),(Item=(Type=Bone,Name="neck_01"),Weight=0.9))
    Weight=1
    Parents.0.Weight=@n3.Value
    Parents.1.Weight=@n4.Value
   )
   > &RigUnit_ParentConstraint(
    AdvancedSettings=(InterpolationType=Average,RotationOrderForFilter=XZY)
    Child=(Type=Bone,Name="dyn_hood_fr_r_corr")
    Parents=((Item=(Type=Bone,Name="head"),Weight=0.7),(Item=(Type=Bone,Name="neck_01"),Weight=0.3))
    Weight=1
   )
   > &RigUnit_ParentConstraint(
    AdvancedSettings=(InterpolationType=Average,RotationOrderForFilter=XZY)
    Child=(Type=Bone,Name="dyn_hood_fr_l_corr")
    Parents=((Item=(Type=Bone,Name="head"),Weight=0.7),(Item=(Type=Bone,Name="neck_01"),Weight=0.3))
    Weight=1
   )
   DeltaRef_rotation_6 = RigVMFunctionReferenceNode(
    Alpha Premult=1
    Node=(Type=Bone,Name="head")
    Parent Initial=false
    RotOrder=ZYX
    altParent=(Type=Bone,Name="neck_02")
    debug=false
    fromSelf=false
    max=-45
    min=0
    twistAxis=(X=0,Y=0,Z=1)
    twistOnly=true
    watchAxis=(X=0,Y=0,Z=1)
    ExecuteContext=@RigVMFunction_Sequence_2.D
   )
   > &RigUnit_ModifyTransforms(
    ItemToModify="((Item=(Type=Bone,Name=\"dyn_hood_bk_up_corr\"),Transform=(Rotation=(X=0,Y=0,Z=0,W=1),Translation=(X=13.6647,Y=0,Z=0),Scale3D=(X=1,Y=1,Z=1))),(Item=(Type=Bone,Name=\"dyn_hood_fr_up_r_corr\"),Transform=(Rotation=(X=0,Y=0,Z=0,W=1),Translation=(X=0,Y=4.8816,Z=0), ...(truncated)"
    WeightMinimum=0
    Weight=@DeltaRef_rotation_6.outAlpha
   )
  ]}
]
variables [
 handChildren {type="TArray<FRigElementKey>", default=((Type=Bone,Name="thumb_01_r"))}
 hoodConstr_val_01 {type=float, default=0.3}
 hoodConstr_val_02 {type=double, default=0.7}
 isFaceacc {type=bool, default=True}
]
